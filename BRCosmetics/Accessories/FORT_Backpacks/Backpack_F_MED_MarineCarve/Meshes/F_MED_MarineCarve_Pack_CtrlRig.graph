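# Control Rig for the MarineCarve backpack cosmetic that nudges the pack bone to suit the wearer's movement state.
# When the skydive-float and tactical-sprint floats sum to at least 1, it applies a sprint-weighted tilt, then a float-weighted offset.
# Otherwise it applies a swim-weighted offset, with all weights passed through AlphaInterp.
# An inverse-weighted tilt step has no execution link, and the roller bones are not driven here.
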ControlRigBlueprint /BRCosmetics/Accessories/FORT_Backpacks/Backpack_F_MED_MarineCarve/Meshes/F_MED_MarineCarve_Pack_CtrlRig
graphs [
 RigVMFunctionLibrary {links [], nodes []}
 RigVMModel {nodes [
   RigUnit_BeginExecution = &RigUnit_BeginExecution()
   Add = &RigVMFunction_MathDoubleAdd(A=RigVMVariableNode() {variable=IsSkyfFoatingFloat}.Value, B=RigVMVariableNode() {variable=isTecticalSprintingFloat}.Value)
   GreaterEqual = /Script/RigVM.RigVMFunction_MathDoubleGreaterEqual(B=1, A=@Add.Result) {is=struct RigVMUnitNode, template="GreaterEqual::Execute(in A,in B,out Result)"}
   RigVMFunction_ControlFlowBranch = &RigVMFunction_ControlFlowBranch(Condition=@GreaterEqual.Result, ExecuteContext=@RigUnit_BeginExecution.ExecutePin)
   n1 = &RigVMFunction_AlphaInterp(
    InterpSpeedDecreasing=5
    InterpSpeedIncreasing=5
    Scale=1
    bClampResult=False
    bInterpResult=False
    Value=RigVMVariableNode() {variable=isTecticalSprintingFloat}.Value
   )
   &RigUnit_ModifyTransforms(
    ItemToModify=((Item=(Type=Bone,Name="pack"),Transform=(Rotation=(X=-0.0436,Y=0,Z=0,W=0.999),Translation=(X=0,Y=0,Z=0),Scale3D=(X=1,Y=1,Z=1))))
    WeightMinimum=0
    Weight=@n1.Result
    ExecutePin=@RigVMFunction_ControlFlowBranch.True
   )
   n2 = &RigVMFunction_AlphaInterp(InterpSpeedDecreasing=8, InterpSpeedIncreasing=5, Scale=1, bClampResult=False, Value=RigVMVariableNode() {variable=IsSkyfFoatingFloat}.Value)
   > &RigUnit_ModifyTransforms(
    ItemToModify=((Item=(Type=Bone,Name="pack"),Transform=(Rotation=(X=0,Y=0,Z=0,W=1),Translation=(X=0,Y=-2,Z=2),Scale3D=(X=1,Y=1,Z=1))))
    WeightMinimum=0
    Weight=@n2.Result
   )
   n3 = &RigVMFunction_AlphaInterp(
    InterpSpeedDecreasing=10
    InterpSpeedIncreasing=10
    Scale=1
    bClampResult=False
    bInterpResult=false
    Value=RigVMVariableNode() {variable=Swim}.Value
   )
   &RigUnit_ModifyTransforms(
    ItemToModify=((Item=(Type=Bone,Name="pack"),Transform=(Rotation=(X=0,Y=0,Z=0,W=1),Translation=(X=0,Y=-2,Z=0),Scale3D=(X=1,Y=1,Z=1))))
    WeightMinimum=0
    Weight=@n3.Result
    ExecutePin=@RigVMFunction_ControlFlowBranch.False
   )
   Subtract = &RigVMFunction_MathDoubleSub(A=1, B=@Add.Result)
   Absolute = &RigVMFunction_MathDoubleAbs(Value=@Subtract.Result)
   AlphaInterp = &RigVMFunction_AlphaInterp(InterpSpeedDecreasing=5, InterpSpeedIncreasing=5, Scale=1, bClampResult=False, bInterpResult=False, Value=@Absolute.Result)
   &RigUnit_ModifyTransforms(
    ExecutePin=()
    ItemToModify=((Item=(Type=Bone,Name="pack"),Transform=(Rotation=(X=-0.0872,Y=0,Z=0,W=0.9962),Translation=(X=0,Y=0.5,Z=0),Scale3D=(X=1,Y=1,Z=1))))
    WeightMinimum=0
    Weight=@AlphaInterp.Result
   )
  ]}
]
hierarchy [
 root {loc=<0 0 0>, rot=<0 0 0 1>}
 pack {parent=root, loc=<0 -0.7316 0>, rot=<0 0 0 1>}
 base_l_roller_02 {parent=pack, loc=<6.5243 -5.5152 2.5815>, rot=<0.2046 -0.3143 0.6375 -0.673>}
 dyn_l_roller_01 {parent=base_l_roller_02, loc=<0 0 0>, rot=<0 0 0 1>}
 dyn_l_roller_l_tyer_01 {parent=dyn_l_roller_01, loc=<2.243 6.9483 1.283>, rot=<0.043 0.0467 -0.7059 -0.7055>}
 dyn_l_roller_r_tyer_02 {parent=dyn_l_roller_01, loc=<2.2138 -6.9834 1.1972>, rot=<0.7057 -0.7056 0.0416 -0.0484>}
 base_r_roller_02 {parent=pack, loc=<-6.3511 -6.087 1.7113>, rot=<-0.2947 0.1846 0.6792 -0.6463>}
 dyn_r_roller_01 {parent=base_r_roller_02, loc=<0 0 0>, rot=<0 0 0 1>}
 dyn_r_roller_l_tyer_01 {parent=dyn_r_roller_01, loc=<2.2124 6.9734 1.2274>, rot=<0.0463 0.0464 -0.7058 -0.7054>}
 dyn_r_roller_r_tyer_02 {parent=dyn_r_roller_01, loc=<2.2264 -6.9612 1.2511>, rot=<0 0 0 1>}
]
variables [
 IsSkyfFoatingFloat {type=double, default=0}
 Swim {type=double, default=0}
 isTecticalSprintingFloat {type=double, default=0}
]
