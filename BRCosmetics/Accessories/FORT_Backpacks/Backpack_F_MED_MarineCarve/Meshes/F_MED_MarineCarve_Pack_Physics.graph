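# Physics asset for the MarineCarve backpack cosmetic mesh, covering the pack body and two dangling roller assemblies, left and right.
# Each body is a single capsule: the pack, each roller, and two tyer capsules per roller.
# Constraints chain the tyers to their roller and each roller to the pack.
# Very tight swing and twist limits keep the rollers and tyers nearly rigid, allowing only slight jiggle.
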PhysicsAsset /BRCosmetics/Accessories/FORT_Backpacks/Backpack_F_MED_MarineCarve/Meshes/F_MED_MarineCarve_Pack_Physics
bodies {
 dyn_l_roller_01 [{shapeType=Capsule, length=7.4156, radius=1.6491, shapeName=dyn_l_roller_01_capsule, center=<2 0 1.3>, rotation=<rot 0 -180 -90>}]
 dyn_l_roller_l_tyer_01 [{shapeType=Capsule, length=0.7054, radius=1.858, shapeName=dyn_l_roller_l_tyer_01_capsule, center=<0.1329 0 0>}]
 dyn_l_roller_r_tyer_02 [{shapeType=Capsule, length=0.8304, radius=1.858, shapeName=dyn_l_roller_l_tyer_01_capsule, center=<0.1329 0 0>}]
 dyn_r_roller_01 [{shapeType=Capsule, length=7.4156, radius=1.6491, shapeName=dyn_l_roller_01_capsule, center=<2 0 1.3>, rotation=<rot 0 -180 -90>}]
 dyn_r_roller_l_tyer_01 [{shapeType=Capsule, length=0.8304, radius=1.858, shapeName=dyn_l_roller_l_tyer_01_capsule, center=<-0.1903 0 0>}]
 dyn_r_roller_r_tyer_02 [{shapeType=Capsule, length=0.8304, radius=1.858, shapeName=dyn_l_roller_l_tyer_01_capsule, center=<-0.1903 0 0>}]
 pack [{shapeType=Capsule, length=13.5197, radius=3.2531, shapeName=pack_capsule, rotation=<rot 90 0.0004 0>}]
}
constraints [
 {bone1Name=dyn_l_roller_l_tyer_01, bone2Name=dyn_l_roller_01, swing1LimitDegrees=1, swing2LimitDegrees=1, twistLimitDegrees=5}
 {bone1Name=dyn_l_roller_r_tyer_02, bone2Name=dyn_l_roller_01, swing1LimitDegrees=1, swing2LimitDegrees=1, twistLimitDegrees=5}
 {bone1Name=dyn_r_roller_l_tyer_01, bone2Name=dyn_r_roller_01, swing1LimitDegrees=1, swing2LimitDegrees=1, twistLimitDegrees=5}
 {bone1Name=dyn_r_roller_r_tyer_02, bone2Name=dyn_r_roller_01, swing1LimitDegrees=1, swing2LimitDegrees=1, twistLimitDegrees=5}
 {bone1Name=dyn_l_roller_01, bone2Name=pack, swing1LimitDegrees=0, swing2LimitDegrees=1, twistLimitDegrees=2}
 {bone1Name=dyn_r_roller_01, bone2Name=pack, swing1LimitDegrees=0, swing2LimitDegrees=1, twistLimitDegrees=2}
]
