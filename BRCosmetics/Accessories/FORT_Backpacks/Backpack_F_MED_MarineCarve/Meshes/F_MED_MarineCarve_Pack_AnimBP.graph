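# Animation Blueprint for the MarineCarve backpack cosmetic, adding procedural motion driven by the wearer's state and spine bend.
# A Control Rig receives skydiving, tactical sprint and swim-sprint floats, then feeds a RigidBody simulation with custom gravity and limits.
# Two additive bone modifiers then adjust the pack bone, the first following spine_05's yaw relative to spine_01, the second using fixed offsets.
# The Event Graph also sets a Pickaxe float that nothing shown reads.
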AnimBlueprint /BRCosmetics/Accessories/FORT_Backpacks/Backpack_F_MED_MarineCarve/Meshes/F_MED_MarineCarve_Pack_AnimBP
target_skeleton=&/F_MED_MarineCarve_Pack_Skeleton
graphs [
 AnimGraph {nodes [
   n1 = Control Rig!(
    Source=Input Pose()
    Alpha
    IsSkyfFoatingFloat:double=Get isSkyDivingFloat() -> (isSkyDivingFloat:double)
    isTecticalSprintingFloat:double=Get isTacticalSprintingFloat() -> (isTacticalSprintingFloat:double)
    Swim:double=Get Swiming() -> (Swiming:double)
   ) -> (Pose) {controlRigAssetReference {blueprintRigClass=&/F_MED_MarineCarve_Pack_CtrlRig.F_MED_MarineCarve_Pack_CtrlRig_C}}
   n2 = RigidBody(ComponentPose=Local To Component(LocalPose=@n1)) {
    bOverrideWorldGravity=true
    overridePhysicsAsset=&/F_MED_MarineCarve_Pack_Physics
    overrideWorldGravity {z=-980}
    simSpaceSettings {maxAngularAcceleration=40, maxAngularVelocity=8, maxLinearAcceleration=3500, maxLinearVelocity=700, velocityScaleZ=0.8, worldAlpha=0.5}
   }
   n3 = "Transform (Modify) Bone - Bone: pack"(
    ComponentPose=@n2
    Translation=Get backpack_translation()
    Rotation=Get backpack_rotator() -> (backpack_rotator:Rotator)
    Scale="0.8,0.8,0.8"
   ) {
    rotationMode=BMM_Additive
    rotationSpace=BCS_BoneSpace
    scaleMode=BMM_Additive
    scaleSpace=BCS_BoneSpace
    translationMode=BMM_Additive
    translationSpace=BCS_BoneSpace
    alphaBoolBlend {blendOption=Linear}
    boneToModify {boneName=pack}
    scale=<0.8 0.8 0.8>
   }
   n4 = "Transform (Modify) Bone - Bone: pack"(ComponentPose=@n3, Translation="0,-5,0", Rotation="0,-10,0") {
    rotationMode=BMM_Additive
    scaleMode=BMM_Additive
    scaleSpace=BCS_BoneSpace
    translationMode=BMM_Additive
    alphaBoolBlend {blendOption=Linear}
    alphaScaleBiasClamp {bClampResult=true, bInterpResult=true, clampMax=0.25, interpSpeedDecreasing=2, interpSpeedIncreasing=20}
    boneToModify {boneName=pack}
    rotation {yaw=-10}
    translation {y=-5}
   }
   Output Pose(Result=Component To Local(ComponentPose=@n4)) {layerGroup=None}
  ]}
 EventGraph {nodes [
   Event BlueprintUpdateAnimation()
   > n1 = Sequence!(execute) -> (then_0, then_1, then_2:Exec, then_3:Exec, then_4:Exec)
   n2 = GetOwnerSkeletalMeshComponent()
   n3 = BreakRotator(InRot=BreakTransform(InTransform=GetDeltaTransformFromRefPose(self=@n2, BoneName=spine_05, BaseName=spine_01)).Rotation)
   > Set backpack_rotator(backpack_rotator:Rotator=MakeRotator(Roll=MapRangeClamped(Value=@n3.Yaw, InRangeA=0, InRangeB=-30, OutRangeB=35))) -> (Output_Get:Rotator)
   > Set backpack_translation(backpack_translation=MakeVector(Y=MapRangeClamped(Value=@n3.Yaw, InRangeA=0, InRangeB=-40, OutRangeB=-7)))
   n4 = Branch(execute=@n1.then_2, Condition=Get bIsSwimSprinting() -> (bIsSwimSprinting:bool))
   > Set Swiming(Swiming:double=1) -> (Output_Get:double)
   Set Swiming(execute=@n4.else, Swiming=0)
   n5 = Branch(execute=@n1.then_3, Condition=Get bIsPlayingMeleeAnim())
   > Set Pickaxe(Pickaxe:double=1) -> (Output_Get:double)
   Set Pickaxe(execute=@n5.else, Pickaxe=0)
   Set isSkyDivingFloat(execute=@n1.then_1, isSkyDivingFloat:double="To Float (Boolean)"(InBool=Get bIsSkyDiving())) -> (Output_Get:double)
   > Set isTacticalSprintingFloat(isTacticalSprintingFloat:double="To Float (Boolean)"(InBool=Get bIsTacticalSprinting())) -> (Output_Get:double)
   TryGetPawnOwner()
   OR Boolean(A=false, B=Get bIsSurfaceSwimming())
  ]}
]
variables [
 backpack_rotator {type=Rotator}
 backpack_translation {type=Vector}
 isSkyDivingFloat {type=number}
 isTacticalSprintingFloat {type=number}
 Pickaxe {type=number}
 Swiming {type=number}
]
